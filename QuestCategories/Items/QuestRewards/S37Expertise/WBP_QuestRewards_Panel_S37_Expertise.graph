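# Quest rewards panel widget for S37 Expertise, showing vending machine, weapon, Nanaimo bar and boon art over glows, with text on seasonal unlocks and daily quests.
# Layout is a fixed-size box of overlaid images positioned by render transforms.
# PreConstruct calls SetMobileStyle, which in compact UI centers six reward images, moves them to variable-driven offsets and resets text overlay padding.
# The FloatAnim playback node is unwired, and the Default animation has no bindings.
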WidgetBlueprint /QuestCategories/Items/QuestRewards/S37Expertise/WBP_QuestRewards_Panel_S37_Expertise
animations [Default {
  binding_count=0
  length_seconds=5
  object=&self:Default
  sequence {bindings []}
 }]
graphs [
 EventGraph {nodes [
   Event PreConstruct()
   > SetMobileStyle() {is=title K2Node_CallFunction}
   PlayAnimation(InAnimation=Get FloatAnim() -> (FloatAnim:Widget Animation Object Reference), NumLoopsToPlay=0)
   Event Construct()
  ]}
 SetMobileStyle {nodes [
   SetMobileStyle() {is=title K2Node_FunctionEntry}
   > Branch(Condition=IsUsingCompactUI())
   n1 = Get Image_Boon_ExtendedMag() -> (Image_Boon_ExtendedMag:Image Object Reference)
   n2 = SlotAsOverlaySlot(Widget=@n1)
   n3 = Get Image_Boon_AgileAiming() -> (Image_Boon_AgileAiming:Image Object Reference)
   n4 = SlotAsOverlaySlot(Widget=@n3)
   n5 = Get Image_VM() -> (Image_VM:Image Object Reference)
   n6 = SlotAsOverlaySlot(Widget=@n5)
   n7 = Get Image_NanaimoBar() -> (Image_NanaimoBar:Image Object Reference)
   n8 = SlotAsOverlaySlot(Widget=@n7)
   n9 = Get Image_HeavyMiniGun() -> (Image_HeavyMiniGun:Image Object Reference)
   n10 = SlotAsOverlaySlot(Widget=@n9)
   n11 = Get Image_ClusterLauncher() -> (Image_ClusterLauncher:Image Object Reference)
   n12 = SlotAsOverlaySlot(Widget=@n11)
   > SetHorizontalAlignment(self=[@n2, @n4, @n6, @n8, @n10, @n12], InHorizontalAlignment=HAlign_Center)
   > SetVerticalAlignment(self=[@n2, @n4, @n6, @n8, @n10, @n12], InVerticalAlignment=VAlign_Center)
   > SetRenderTranslation(self=@n3, Translation=Get Image1_CompactPosition() -> (Image1_CompactPosition:Vector 2D Structure))
   > SetRenderTranslation(self=@n1, Translation=Get Image2_CompactPosition() -> (Image2_CompactPosition:Vector 2D Structure))
   > SetRenderTranslation(self=@n11, Translation=Get Image3_CompactPosition() -> (Image3_CompactPosition:Vector 2D Structure))
   > SetRenderTranslation(self=@n9, Translation=Get Image4_CompactPosition() -> (Image4_CompactPosition:Vector 2D Structure))
   > SetRenderTranslation(self=@n7, Translation=Get Image5_CompactPosition() -> (Image5_CompactPosition:Vector 2D Structure))
   > SetRenderTranslation(self=@n5, Translation=Get Image6_CompactPosition() -> (Image6_CompactPosition:Vector 2D Structure))
   > SetPadding!(
    execute
    self:Overlay Slot Object Reference=SlotAsOverlaySlot(Widget=Get Overlay_1() -> (Overlay_1:Overlay Object Reference))
    InPadding:Margin Structure
   ) -> (then)
  ]}
]
variables [
 Image1_CompactPosition {type=Vector2D, default=<225 0>}
 Image2_CompactPosition {type=Vector2D, default=<-225 0>}
 Image3_CompactPosition {type=Vector2D, default=<0 -325>}
 Image4_CompactPosition {type=Vector2D, default=<-200 -200>}
 Image5_CompactPosition {type=Vector2D, default=<200 -250>}
 Image6_CompactPosition {type=Vector2D, default=<0 0>}
]
widgets {
 info {inheritedWidgetCount=0, parentClass=&UserWidget, rootWidgetClass=&SizeBox, widgetCount=13}
 widgets [
  SizeBox_0 {widgetClassPath=&SizeBox, widget=SizeBox_0, properties {bOverride_WidthOverride=true, heightOverride=940, widthOverride=850}}
  Overlay_0 {widgetClassPath=&Overlay, parent=SizeBox_0, widget=Overlay_0, slot {class=SizeBoxSlot}}
  Image_BackgroundGlow {
   widgetClassPath=&Image
   bIsVariable=true
   parent=Overlay_0
   widget=Image_BackgroundGlow
   properties {
    brush {
     bIsDynamicallyLoaded=false
     drawAs=Image
     imageType=FullColor
     mirroring=NoMirror
     resourceName=None
     resourceObject=/QuestCategories/UI/RewardPanel/Materials/MI_UI_QuestRewards_Glow_S37_Expertise
     tiling=NoTile
     imageSize=<1024 1024>
     margin=<ltrb 0 0 0 0>
     outlineSettings {
      bUseBrushTransparency=false
      roundingType=HalfHeightRadius
      width=0
      color {colorUseRule=UseColor_Specified, specifiedColor=<rgba 0 0 0 0>}
      cornerRadii=<0 0 0 0>
     }
     tintColor {colorUseRule=UseColor_Specified, specifiedColor=<rgba 1 1 1 1>}
     uVRegion {bIsValid=false, max=<0 0>, min=<0 0>}
    }
    renderTransform {angle=0, scale=<2.5 1.5>, shear=<0 0>, translation=<100 0>}
   }
   slot {class=OverlaySlot}
  }
  Image_VM {
   widgetClassPath=&Image
   bIsVariable=true
   parent=Overlay_0
   widget=Image_VM
   properties {
    brush {
     bIsDynamicallyLoaded=false
     drawAs=Image
     imageType=FullColor
     mirroring=NoMirror
     resourceName=None
     resourceObject=&/T_UI_QuestReward_VendingMachine_Expertise
     tiling=NoTile
     imageSize=<512 512>
     margin=<ltrb 0 0 0 0>
     outlineSettings {
      bUseBrushTransparency=false
      roundingType=HalfHeightRadius
      width=0
      color {colorUseRule=UseColor_Specified, specifiedColor=<rgba 0 0 0 0>}
      cornerRadii=<0 0 0 0>
     }
     tintColor {colorUseRule=UseColor_Specified, specifiedColor=<rgba 1 1 1 1>}
     uVRegion {bIsValid=false, max=<0 0>, min=<0 0>}
    }
    renderTransform {angle=0, scale=<1.2 1.2>, shear=<0 0>, translation=<160 200>}
   }
   slot {class=OverlaySlot}
  }
  Image_HeavyMiniGun {
   widgetClassPath=&Image
   bIsVariable=true
   parent=Overlay_0
   widget=Image_HeavyMiniGun
   properties {
    brush {
     bIsDynamicallyLoaded=false
     drawAs=Image
     imageType=FullColor
     mirroring=NoMirror
     resourceName=None
     resourceObject=&/T_UI_QuestReward_HeavyMiniGun_TidalCrane
     tiling=NoTile
     imageSize=<256 256>
     margin=<ltrb 0 0 0 0>
     outlineSettings {
      bUseBrushTransparency=false
      roundingType=HalfHeightRadius
      width=0
      color {colorUseRule=UseColor_Specified, specifiedColor=<rgba 0 0 0 0>}
      cornerRadii=<0 0 0 0>
     }
     tintColor {colorUseRule=UseColor_Specified, specifiedColor=<rgba 1 1 1 1>}
     uVRegion {bIsValid=false, max=<0 0>, min=<0 0>}
    }
    renderTransform {angle=-15, scale=<1 1>, shear=<0 0>, translation=<66 140>}
   }
   slot {class=OverlaySlot}
  }
  Image_ClusterLauncher {
   widgetClassPath=&Image
   bIsVariable=true
   parent=Overlay_0
   widget=Image_ClusterLauncher
   properties {
    brush {
     bIsDynamicallyLoaded=false
     drawAs=Image
     imageType=FullColor
     mirroring=NoMirror
     resourceName=None
     resourceObject=&/T_UI_QuestReward_RocketLauncher_TidalCrane
     tiling=NoTile
     imageSize=<256 256>
     margin=<ltrb 0 0 0 0>
     outlineSettings {
      bUseBrushTransparency=false
      roundingType=HalfHeightRadius
      width=0
      color {colorUseRule=UseColor_Specified, specifiedColor=<rgba 0 0 0 0>}
      cornerRadii=<0 0 0 0>
     }
     tintColor {colorUseRule=UseColor_Specified, specifiedColor=<rgba 1 1 1 1>}
     uVRegion {bIsValid=false, max=<0 0>, min=<0 0>}
    }
    renderTransform {angle=0, scale=<1 1>, shear=<0 0>, translation=<300 0>}
   }
   slot {class=OverlaySlot}
  }
  Image_NanaimoBar {
   widgetClassPath=&Image
   bIsVariable=true
   parent=Overlay_0
   widget=Image_NanaimoBar
   properties {
    brush {
     bIsDynamicallyLoaded=false
     drawAs=Image
     imageType=FullColor
     mirroring=NoMirror
     resourceName=None
     resourceObject=&/T_UI_QuestReward_NanaimoBar
     tiling=NoTile
     imageSize=<256 256>
     margin=<ltrb 0 0 0 0>
     outlineSettings {
      bUseBrushTransparency=false
      roundingType=HalfHeightRadius
      width=0
      color {colorUseRule=UseColor_Specified, specifiedColor=<rgba 0 0 0 0>}
      cornerRadii=<0 0 0 0>
     }
     tintColor {colorUseRule=UseColor_Specified, specifiedColor=<rgba 1 1 1 1>}
     uVRegion {bIsValid=false, max=<0 0>, min=<0 0>}
    }
    renderTransform {angle=45, scale=<1 1>, shear=<0 0>, translation=<-80 94>}
   }
   slot {class=OverlaySlot, properties {horizontalAlignment=HAlign_Right}}
  }
  Image_Boon_AgileAiming {
   widgetClassPath=&Image
   bIsVariable=true
   parent=Overlay_0
   widget=Image_Boon_AgileAiming
   properties {
    brush {
     bIsDynamicallyLoaded=false
     drawAs=Image
     imageType=FullColor
     mirroring=NoMirror
     resourceName=None
     resourceObject=&/T_UI_Boon_AgileAiming
     tiling=NoTile
     imageSize=<128 128>
     margin=<ltrb 0 0 0 0>
     outlineSettings {
      bUseBrushTransparency=false
      roundingType=HalfHeightRadius
      width=0
      color {colorUseRule=UseColor_Specified, specifiedColor=<rgba 0 0 0 0>}
      cornerRadii=<0 0 0 0>
     }
     tintColor {colorUseRule=UseColor_Specified, specifiedColor=<rgba 1 1 1 1>}
     uVRegion {bIsValid=false, max=<0 0>, min=<0 0>}
    }
    renderTransform {angle=0, scale=<1 1>, shear=<0 0>, translation=<600 460>}
   }
   slot {class=OverlaySlot}
  }
  Image_Boon_ExtendedMag {
   widgetClassPath=&Image
   bIsVariable=true
   parent=Overlay_0
   widget=Image_Boon_ExtendedMag
   properties {
    brush {
     bIsDynamicallyLoaded=false
     drawAs=Image
     imageType=FullColor
     mirroring=NoMirror
     resourceName=None
     resourceObject=&/T_UI_Boon_ExtendedMag
     tiling=NoTile
     imageSize=<128 128>
     margin=<ltrb 0 0 0 0>
     outlineSettings {
      bUseBrushTransparency=false
      roundingType=HalfHeightRadius
      width=0
      color {colorUseRule=UseColor_Specified, specifiedColor=<rgba 0 0 0 0>}
      cornerRadii=<0 0 0 0>
     }
     tintColor {colorUseRule=UseColor_Specified, specifiedColor=<rgba 1 1 1 1>}
     uVRegion {bIsValid=false, max=<0 0>, min=<0 0>}
    }
    renderTransform {angle=0, scale=<1 1>, shear=<0 0>, translation=<112 460>}
   }
   slot {class=OverlaySlot}
  }
  Overlay_1 {
   widgetClassPath=&Overlay
   bIsVariable=true
   parent=Overlay_0
   widget=Overlay_1
   slot {class=OverlaySlot, properties {horizontalAlignment=HAlign_Fill, verticalAlignment=VAlign_Bottom, padding=<ltrb 0 40 0 40>}}
  }
  Border_53 {
   widgetClassPath=&Border
   parent=Overlay_1
   widget=Border_53
   properties {
    background {
     bIsDynamicallyLoaded=false
     drawAs=Image
     imageType=FullColor
     mirroring=NoMirror
     resourceName=None
     resourceObject=/QuestCategories/UI/RewardPanel/Materials/MI_UI_QuestRewards_InfoBackground_ForbiddenFruit_Eco_01
     tiling=NoTile
     imageSize=<32 32>
     margin=<ltrb 0 0 0 0>
     outlineSettings {
      bUseBrushTransparency=false
      roundingType=HalfHeightRadius
      width=0
      color {colorUseRule=UseColor_Specified, specifiedColor=<rgba 0 0 0 0>}
      cornerRadii=<0 0 0 0>
     }
     tintColor {colorUseRule=UseColor_Specified, specifiedColor=<rgba 1 1 1 1>}
     uVRegion {bIsValid=false, max=<0 0>, min=<0 0>}
    }
    padding=<ltrb 50 20 50 20>
   }
   slot {class=OverlaySlot, properties {horizontalAlignment=HAlign_Center, verticalAlignment=VAlign_Bottom}}
  }
  RichText_QRCode {
   widgetClassPath=&CommonRichTextBlock
   bIsVariable=true
   parent=Border_53
   widget=RichText_QRCode
   properties {
    autoWrapText=true
    renderOpacity=0.7
    text="Unlock your choice of exclusive Expertise Vending Machine sales for the rest of the season!\r\n\r\nDaily Expertise Quests expire each day to roll 3 new ones. "
    textStyleSet=/QuestCategories/UI/RewardPanel/DT_UI_RichTextStyles_QuestRewards_ForbiddenFruit_Eco_01
    decoratorClasses [/Game/UI/Sidebar/RebootRally/BP_UI_RebootRallyRichImageDecorator.BP_UI_RebootRallyRichImageDecorator_C]
    defaultTextStyleOverride {
     overflowPolicy=Clip
     transformPolicy=None
     colorAndOpacity {colorUseRule=UseColor_Specified, specifiedColor=<rgba 1 1 1 1>}
     font {
      bForceMonospaced=false
      bMaterialIsStencil=false
      fontMaterial=null
      fontObject=&BurbankSmall
      letterSpacing=0
      monospacedWidth=1
      size=19
      skewAmount=0
      typefaceFontName=Bold
      outlineSettings {
       bApplyOutlineToDropShadows=false
       bMiteredCorners=false
       bSeparateFillAlpha=false
       outlineMaterial=null
       outlineSize=0
       outlineColor=<rgba 0 0 0 1>
      }
     }
     highlightColor {colorUseRule=UseColor_Specified, specifiedColor=<rgba 0 0 0 1>}
     highlightShape {
      bIsDynamicallyLoaded=false
      drawAs=Image
      imageType=NoImage
      mirroring=NoMirror
      resourceName=None
      resourceObject=null
      tiling=NoTile
      imageSize=<32 32>
      margin=<ltrb 0 0 0 0>
      outlineSettings {
       bUseBrushTransparency=false
       roundingType=HalfHeightRadius
       width=0
       color {colorUseRule=UseColor_Specified, specifiedColor=<rgba 0 0 0 0>}
       cornerRadii=<0 0 0 0>
      }
      tintColor {colorUseRule=UseColor_Specified, specifiedColor=<rgba 1 1 1 1>}
      uVRegion {bIsValid=false, max=<0 0>, min=<0 0>}
     }
     selectedBackgroundColor {colorUseRule=UseColor_Foreground, specifiedColor=<rgba 1 0 1 1>}
     shadowColorAndOpacity=<rgba 0 0 0 1>
     shadowOffset=<0 0>
     strikeBrush {
      bIsDynamicallyLoaded=false
      drawAs=Image
      imageType=NoImage
      mirroring=NoMirror
      resourceName=None
      resourceObject=null
      tiling=NoTile
      imageSize=<32 32>
      margin=<ltrb 0 0 0 0>
      outlineSettings {
       bUseBrushTransparency=false
       roundingType=HalfHeightRadius
       width=0
       color {colorUseRule=UseColor_Specified, specifiedColor=<rgba 0 0 0 0>}
       cornerRadii=<0 0 0 0>
      }
      tintColor {colorUseRule=UseColor_Specified, specifiedColor=<rgba 1 1 1 1>}
      uVRegion {bIsValid=false, max=<0 0>, min=<0 0>}
     }
     underlineBrush {
      bIsDynamicallyLoaded=false
      drawAs=Image
      imageType=NoImage
      mirroring=NoMirror
      resourceName=None
      resourceObject=null
      tiling=NoTile
      imageSize=<32 32>
      margin=<ltrb 0 0 0 0>
      outlineSettings {
       bUseBrushTransparency=false
       roundingType=HalfHeightRadius
       width=0
       color {colorUseRule=UseColor_Specified, specifiedColor=<rgba 0 0 0 0>}
       cornerRadii=<0 0 0 0>
      }
      tintColor {colorUseRule=UseColor_Specified, specifiedColor=<rgba 1 1 1 1>}
      uVRegion {bIsValid=false, max=<0 0>, min=<0 0>}
     }
    }
   }
   slot {class=BorderSlot, properties {padding=<ltrb 50 20 50 20>}}
  }
  Image_ForegroundGlow {
   widgetClassPath=&Image
   bIsVariable=true
   parent=Overlay_0
   widget=Image_ForegroundGlow
   properties {
    renderOpacity=0.02
    brush {
     bIsDynamicallyLoaded=false
     drawAs=Image
     imageType=FullColor
     mirroring=NoMirror
     resourceName=None
     resourceObject=/QuestCategories/UI/RewardPanel/Materials/MI_UI_QuestRewards_ForegroundGlow_ForbiddenFruit_Eco_01
     tiling=NoTile
     imageSize=<1024 1024>
     margin=<ltrb 0 0 0 0>
     outlineSettings {
      bUseBrushTransparency=false
      roundingType=HalfHeightRadius
      width=0
      color {colorUseRule=UseColor_Specified, specifiedColor=<rgba 0 0 0 0>}
      cornerRadii=<0 0 0 0>
     }
     tintColor {colorUseRule=UseColor_Specified, specifiedColor=<rgba 1 1 1 1>}
     uVRegion {bIsValid=false, max=<0 0>, min=<0 0>}
    }
    colorAndOpacity=<rgba 1 1 1 0.05>
    renderTransform {angle=0, scale=<1 1>, shear=<0 0>, translation=<1 0>}
   }
   slot {class=OverlaySlot, properties {horizontalAlignment=HAlign_Center, verticalAlignment=VAlign_Center}}
  }
 ]
}
